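# Editor utility Blueprint with a single function, GetMIDFromActor, which fetches a dynamic material instance from a selected building actor.
# The function casts one selected actor to BuildingSMActor, then passes one of its static mesh components to GetDynamicMaterialInstance for the first material slot.
# The result is stored locally; a step to read its parent material is unwired, nothing is returned, and the event graph is empty.
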EditorUtilityBlueprint /Game/NeverCook/EditorUtilities/Environment/MICUtility/AAU_MICTOMID
graphs [
 EventGraph {links [], nodes []}
 GetMIDFromActor {nodes [
   GetMIDFromActor() {is=title K2Node_FunctionEntry}
   > n1 = GetSelectedActors()
   > n2 = Cast To BuildingSMActor!(execute, Object="Get (a copy)"(Array:Array of Actor Object References=@n1) -> (Output:Actor Object Reference)) -> (then, CastFailed, AsBuilding Static Mesh:Building Static Mesh Object Reference) {is=title K2Node_DynamicCast}
   n3 = Get Components By Class!(self:Actor Object Reference=@n2, ComponentClass:Actor Component Class Reference=&StaticMeshComponent) -> (ReturnValue:Array of Static Mesh Component Object References)
   > Set L_DynamicMaterial(L_DynamicMaterial:Material Instance Dynamic Object Reference=GetDynamicMaterialInstance!(
     MeshComponent:Mesh Component Object Reference="Get (a copy)"(Array:Array of Static Mesh Component Object References=@n3) -> (Output:Static Mesh Component Object Reference)
     MaterialIndex:int=0
     __WorldContext:Object Reference=@n2
    ) -> (Material Instance Dynamic:Material Instance Dynamic Object Reference)) -> (Output_Get:Material Instance Dynamic Object Reference)
   Set L_MaterialParent(L_MaterialParent:Material Interface Object Reference=Get Parent!(self:Material Instance Object Reference=Get L_DynamicMaterial() -> (L_DynamicMaterial:Material Instance Dynamic Object Reference)) -> (Parent:Material Interface Object Reference)) -> (Output_Get:Material Interface Object Reference)
  ]}
]
variables []
